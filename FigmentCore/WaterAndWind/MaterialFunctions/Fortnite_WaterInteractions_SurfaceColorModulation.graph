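# Material function for Fortnite wind and water, built to modulate surface Diffuse color with player water interactions, though its output passes Diffuse through unchanged.
# The unconnected modulation chain scales a height map, by default from SampleRealTimeWaterData, by a parameter, clamps it and feeds MultiplyAdd with Diffuse.
# A static switch toggles that result, then RenderToTextureEnabled selects it or plain Diffuse according to render-to-texture support.
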MaterialFunction /FigmentCore/WaterAndWind/MaterialFunctions/Fortnite_WaterInteractions_SurfaceColorModulation
graph {
 properties {libraryCategoriesText [Fortnite Wind and Water]}
 nodes [
  n1 = FunctionInput() {inputName=Diffuse}
  FunctionOutput(None=@n1)
  n2 = MaterialFunctionCall!(Normal intensity, "World space offset to calculate normal ") {materialFunction=&/Fortnite_SampleRealTimeWaterData}
  n3 = Clamp(None=Multiply(
    A=FunctionInput(Preview=@n2.Height Map) {bUsePreviewValueAsDefault=true, inputName=HeightMap, inputType=FunctionInput_Scalar}
    B=ScalarParameter() {parameterName=Base Color Player Interaction Modulation}
   )) {maxDefault=2}
  n4 = StaticBoolParameter() {defaultValue=true, group=Player interaction, parameterName=Use Player Interaction Base Color Modulation}
  MaterialFunctionCall!(
   With RTT Features=StaticSwitch(
    True=MaterialFunctionCall!(Base=@n1, Add=@n3) {materialFunction=/Engine/Functions/Engine_MaterialFunctions02/Math/MultiplyAdd}.Result
    False=@n1
    Value=FunctionInput(Preview=@n4) {bUsePreviewValueAsDefault=true, inputName=Use Player Interaction Base Color Modulation, inputType=FunctionInput_StaticBool}
   )
   WithOut RTT Features=@n1
   With RTT Features V2
   WithOut RTT Features V2
   With RTT Features Att
   WithOut RTT Features Att
  ) {materialFunction=/FigmentCore/WaterAndWind/EnableOrDisableWind/Fortnite_RenderToTextureEnabled}
 ]
}
