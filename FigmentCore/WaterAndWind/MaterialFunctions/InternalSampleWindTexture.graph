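# Material function in the Fortnite Wind and Water library that samples a wind render target at a world position.
# UVs come from the XY offset to a texel-specific camera position, defaulting to a parameter collection value, divided by OrthoWidth and shifted by half.
# It outputs RGB, alpha, UVs and a border mask that darkens steeply toward the edges.
# The mask's Max takes R twice, so only the U axis affects it.
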MaterialFunction /FigmentCore/WaterAndWind/MaterialFunctions/InternalSampleWindTexture
graph {
 properties {libraryCategoriesText [Fortnite Wind and Water]}
 nodes [
  n1 = CollectionParameter() {
   collection=&FortniteMaterialParameters
   parameterName=WindSpecificCameraPositionVariable1_DoNotUseElsewhere
  }
  n2 = Subtract(
   A=FunctionInput(Preview=WorldPosition().XYZ) {bUsePreviewValueAsDefault=true, inputName=Sample Point World Position}
   B=ComponentMask(None=FunctionInput(Preview=@n1) {bUsePreviewValueAsDefault=true, inputName=Texel Specific Camera Position, inputType=FunctionInput_Vector4}) {b=true, g=true, r=true}
  )
  n3 = FunctionInput() {bUsePreviewValueAsDefault=true, inputName=OrthoWidth, inputType=FunctionInput_Scalar, previewValue=<2048 0 0 1>}
  n4 = Subtract(A=Divide(A=ComponentMask(None=@n2) {g=true, r=true}, B=@n3) {constB=1024}) {constB=-0.5}
  n5 = TextureSample(
   UVs=@n4
   Tex=FunctionInput(Preview=TextureObject() {texture=/FigmentCore/WaterAndWind/Textures/WindRenderTarget1}) {bUsePreviewValueAsDefault=true, inputName=Wind Texture, inputType=FunctionInput_Texture2D, sortPriority=2}
  ) {texture=&WindRenderTarget1}
  FunctionOutput(None=@n5.A) {outputName=Alpha}
  FunctionOutput(None=@n5.RGB) {outputName=RGB}
  FunctionOutput(None=@n4) {outputName=UVs}
  n6 = ConstantBiasScale(None=Clamp(None=@n4)) {bias=-0.5, scale=2}
  n7 = MaterialFunctionCall(Float2=Power(Base=Abs(None=@n6)) {constExponent=15}) {materialFunction=&BreakOutFloat2Components}
  FunctionOutput(None=OneMinus(None=Max(A=@n7.R, B=@n7.R))) {description=Black outside, outputName=Border Mask}
  MaterialFunctionCall() {materialFunction=/Engine/Functions/Engine_MaterialFunctions02/PivotPainter/PivotPainter_PerObjectData}
 ]
}
